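# Alpha-composite deferred decal material for IronLog's toon paint splats, varying per placement via actor position.
# That position picks a frame from 4x4 mask and normal flipbooks and offsets world-space Perlin-Worley noise driving curve-atlas colors.
# Mask thresholds add outlines, sky-light dot normal gives contrast-adjustable shading, and time-of-day tints follow.
# Decal lifetime drives noise dissolve alongside a facing mask; a cubemap/gradient reflection branch feeds emissive adjustments.
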Material /WildEstateWeapons/Gameplay/IronLog/FX/Decals/MD_ToonSPlat
graph {
 properties {bAutomaticallySetUsageInEditor=false, blendMode=BLEND_AlphaComposite, materialDomain=MD_DeferredDecal, usedShadingModels=Default Lit}
 nodes [
  Constant3Vector()
  Constant3Vector()
  Constant()
  Constant() {r=1}
  Constant()
  MaterialFunctionCall() {materialFunction=&DebugTimeSine}
  n1 = MaterialFunctionCall(Float3=ActorPositionWS()) {materialFunction=&BreakOutFloat3Components}
  n2 = Add(A=@n1.R, B=@n1.R)
  n3 = LinearInterpolate(Alpha=Frac(None=Add(A=@n2, B=@n1.B))) {constA=0.75}
  n4 = Constant() {r=4}
  n5 = TextureObject() {samplerType=SAMPLERTYPE_Normal, texture=&/Substance_graph_normal}
  n6 = Transform(None=MaterialFunctionCall(
    Normal=MaterialFunctionCall("Animation  Phase (0-1)"=@n3, Number of Rows=@n4, Number of Columns=@n4, Texture=@n5) {materialFunction=&FlipBook}.Result
    Flatness=ScalarParameter() {parameterName=Normal Flatness}
   ) {materialFunction=&FlattenNormal}.Result)
  n7 = ReflectionVectorWS(CustomWorldNormal=@n6)
  n8 = TextureSampleParameterCube!(UVs=@n7, Bias=ScalarParameter() {parameterName=mip bia}, Apply View MipBias) {
   mipValueMode=TMVM_MipBias
   parameterName=Cubemap
   samplerType=SAMPLERTYPE_LinearColor
   texture=/Game/Packages/Fortress_Sky/TexturesHDR/T_AthenaSkylight
  }
  n9 = OneMinus(None=ComponentMask(None=MaterialFunctionCall!(Vector=@n7) {materialFunction=/Engine/Functions/Engine_MaterialFunctions01/Coordinates/LongLatToUV}.Result) {g=true})
  n10 = VectorParameter() {parameterName=Reflection grad options, channelNames {a=contrast, b=add, g=mul, r=pow}, defaultValue=<rgba 1 1 0 0>}
  n11 = Add(A=Multiply(A=Power(Base=@n9, Exp=@n10.pow), B=@n10.mul), B=@n10.add)
  n12 = VectorParameter() {parameterName=Reflect texture tint / desat}
  n13 = Desaturation(
   None=LinearInterpolate(
    A=@n8.RGB
    B=MaterialFunctionCall(In=@n11, Contrast=@n10.contrast) {materialFunction=&CheapContrast}.Result
    Alpha=ScalarParameter() {parameterName=Use refect textur}
   )
   Fraction=@n12.A
  )
  MaterialFunctionCall(Emissive=Multiply(A=Multiply(A=@n13, B=@n12.RGB), B=ScalarParameter() {defaultValue=1, parameterName=Emissmul})) {materialFunction=&MF_FX_EmissiveAdjustments}
  MaterialFunctionCall() {materialFunction=&ComponentPivotLocation}
  MaterialExpressionFrac_2 = Frac(None=ActorPositionWS())
  n14 = Convert(None=@MaterialExpressionFrac_2) {
   convertInputs [{
     type=Vector3
     expressionInput {
      expression=&self:MaterialExpressionFrac_2
      inputName=None
      mask=0
      maskA=0
      maskB=0
      maskG=0
      maskR=0
      outputIndex=0
     }
    }]
   convertOutputs [{type=Vector3}, {type=Vector3}, {type=Vector3}]
  }
  n15 = WorldPosition()
  n16 = Multiply(A=VectorParameter() {parameterName=test, defaultValue=<rgba 0 0 0 1>}.RGB, B=@n15.XYZ)
  n17 = ActorPositionWS()
  n18 = Add(A=@n15.XYZ, B=OneMinus(None=@n17))
  n19 = Divide(A=Distance(A=@n15.XYZ, B=@n17), B=ScalarParameter() {defaultValue=128, parameterName=Collapse Size})
  n20 = Saturate(None=@n19)
  n21 = LinearInterpolate(
   A=ScalarParameter() {defaultValue=1, parameterName=Collaspe Falloff Inner}
   B=ScalarParameter() {defaultValue=1, parameterName=Collaspe Falloff Outer}
   Alpha=@n20
  )
  n22 = Subtract(A=Saturate(None=Power(Base=@n19, Exp=@n21)))
  n23 = ScalarParameter() {defaultValue=328.6945, parameterName=World Pos Density}
  n24 = Divide(
   A=LinearInterpolate(A=@n15.XYZ, B=LinearInterpolate(A=@n16, B=@n18, Alpha=@n22), Alpha=ScalarParameter() {parameterName=Use Collapse})
   B=@n23
  ) {constB=1024}
  n25 = TextureSample!(UVs=Add(A=@n14, B=@n24), Tex, Bias=NamedRerouteUsage(), Apply View MipBias) {mipValueMode=TMVM_MipBias, samplerType=SAMPLERTYPE_LinearColor, texture=/Game/TimeOfDay/CloudsAndStorm/Textures/VT_PerlinWorley}
  n26 = Divide(A=@n24) {constB=2.2652}
  n27 = Multiply(A=@n23) {constB=1.885}
  n28 = TextureSample(UVs=Add(A=@n14, B=Add(A=@n26, B=@n27)), Bias=NamedRerouteUsage()) {mipValueMode=TMVM_MipBias, samplerType=SAMPLERTYPE_LinearColor, texture=&VT_PerlinWorley}
  n29 = Divide(A=@n24) {constB=2.7559}
  n30 = Multiply(A=@n23) {constB=3.3257}
  n31 = TextureSample(UVs=Add(A=@n14, B=Add(A=@n29, B=@n30)), Bias=NamedRerouteUsage()) {mipValueMode=TMVM_MipBias, samplerType=SAMPLERTYPE_LinearColor, texture=&VT_PerlinWorley}
  n32 = Multiply(A=Add(A=Add(A=@n25.R, B=@n28.G), B=@n31.B), B=ScalarParameter() {defaultValue=1.25, parameterName=CurveDencity}) {constB=10}
  n33 = ScalarParameter() {parameterName=Curve offset}
  n34 = CurveAtlasRowParameter(CurveTime=Frac(None=Add(A=@n32, B=@n33))) {
   atlas=&/CA_IronLog
   curve=&/CC_COlors
   parameterName=Color Curve
  }
  n35 = Constant() {r=4}
  n36 = MaterialFunctionCall(
   "Animation  Phase (0-1)"=@n3
   Number of Rows=@n35
   Number of Columns=@n35
   Texture=TextureObject() {samplerType=SAMPLERTYPE_LinearGrayscale, texture=/WildEstateWeapons/Gameplay/IronLog/FX/CameraLens/T_SplatMasks}
  ) {materialFunction=&FlipBook}
  n37 = ScalarParameter() {parameterName=Outline Offset}
  n38 = LinearInterpolate(A=VectorParameter() {parameterName=OutLine Color}.RGB, B=@n34, Alpha=Round(None=Subtract(A=@n36.Result, B=@n37)))
  n39 = Add(A=DotProduct(A=SkyAtmosphereLightDirection(), B=@n6), B=ScalarParameter() {parameterName=Shadow Offset}) {constB=0}
  n40 = ScalarParameter() {parameterName=shadow contrast}
  n41 = LinearInterpolate(
   A=Multiply(A=@n38, B=ScalarParameter() {parameterName=color shadow multiply}) {constB=0.5}
   B=@n38
   Alpha=Saturate(None=MaterialFunctionCall(In=@n39, Contrast=@n40) {materialFunction=&CheapContrast}.Result)
  )
  n42 = MaterialFunctionCall(
   Color_Morning=VectorParameter() {parameterName=moning tint, defaultValue=<rgba 1 1 1 0>}.RGB
   Color_Day=VectorParameter() {parameterName=Day tint, defaultValue=<rgba 1 1 1 0>}.RGB
   Color_Evening=VectorParameter() {parameterName=evening tint, defaultValue=<rgba 1 1 1 0>}.RGB
   Color_Night=VectorParameter() {parameterName=Night tint, defaultValue=<rgba 1 1 1 0>}.RGB
  ) {materialFunction=&Color_Time_of_Day}
  n43 = VectorParameter() {parameterName=Dissolve tile contrast, channelNames {a=Na, b=Contrast, g=tile Y, r=tile X}, defaultValue=<rgba 1 1 1 0>}
  n44 = Multiply(A=ComponentMask(None=@n43.RGB) {g=true, r=true}, B=TextureCoordinate())
  n45 = DecalLifetimeOpacity()
  n46 = Saturate(None=Add(
    A=TextureSample!(UVs=@n44, Tex, Apply View MipBias) {texture=/Game/Global/Textures/Default/Noise/3xPerlin_M}.R
    B=LinearInterpolate(Alpha=@n45.Opacity) {constA=-1}
   ))
  n47 = MaterialFunctionCall() {materialFunction=/Game/Weapons/Prototype/SprayGun/Materials/MF_DecalFacingMask}
  n48 = VectorParameter() {parameterName=Decal Mask Options, defaultValue=<rgba 1 1 0.45 10>}
  n49 = Multiply(A=Power(Base=@n36.Result, Exp=@n48.R), B=@n48.G)
  n50 = MaterialFunctionCall(In=Saturate(None=Add(A=@n49, B=@n48.B)), Contrast=@n48.A) {materialFunction=&CheapContrast}
  Multiply(
   A=Multiply(A=Multiply(A=@n41, B=@n42.Result), B=ScalarParameter() {parameterName=base color Mul})
   B=Multiply(
    A=MaterialFunctionCall(In=@n46, Contrast=@n43.Contrast) {materialFunction=&CheapContrast}.Result
    B=Multiply(A=@n47.Result, B=@n50.Result) {constA=1}
   )
  )
  NamedRerouteDeclaration(None=ScalarParameter() {defaultValue=1, parameterName=MipBias}) {name=Param MipBias, nodeColor=<rgba 0.875 0 1 1>}
  Round(None=@n20)
 ]
}
